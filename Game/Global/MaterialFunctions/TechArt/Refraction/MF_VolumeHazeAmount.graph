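# MF_VolumeHazeAmount is a material function that turns a squared roughness input into a scalar haze amount for refraction materials.
# It saturates Roughness^2 and raises it to a tunable exponent to get a blend weight.
# That weight interpolates between minimum and maximum haze amounts, each clamped to be non-negative.
# The exponent and both limits are optional inputs that fall back to preview values.
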MaterialFunction /Game/Global/MaterialFunctions/TechArt/Refraction/MF_VolumeHazeAmount
graph {nodes [
  n1 = FunctionInput() {bUsePreviewValueAsDefault=true, inputName=Haze Min Amount, inputType=FunctionInput_Scalar, sortPriority=20}
  n2 = FunctionInput() {bUsePreviewValueAsDefault=true, inputName=Haze Max Amount, inputType=FunctionInput_Scalar, sortPriority=30, previewValue=<1 0 0 1>}
  n3 = Saturate(None=FunctionInput() {inputName="Roughness^2", inputType=FunctionInput_Scalar, previewValue=<0.02 0 0 1>})
  n4 = FunctionInput() {bUsePreviewValueAsDefault=true, inputName=Roughness Exponent, inputType=FunctionInput_Scalar, sortPriority=10, previewValue=<2.5 0 0 1>}
  FunctionOutput(None=LinearInterpolate(A=Max(A=@n1) {constB=0}, B=Max(A=@n2) {constB=0}, Alpha=Power(Base=@n3, Exp=@n4) {constExponent=2.5})) {outputName="Haze Amount (S)"}
 ]}
