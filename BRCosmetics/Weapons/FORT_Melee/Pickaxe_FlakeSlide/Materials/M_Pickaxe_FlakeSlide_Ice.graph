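# Ice material for the FlakeSlide pickaxe cosmetic, with CubeIce and TailIce material attribute branches and mask-texture blending.
# A refractive volume simulates ice depth from normal maps, scene thickness, absorption, scattering and haze, with animated parallax noise varying the interior.
# Scene luminance uses either eye-adaptation-corrected scene color or an albedo-and-lighting approximation for alpha blending.
# The graph also uses shared glass, thin-film, camera-dither and character-opacity functions.
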Material /BRCosmetics/Weapons/FORT_Melee/Pickaxe_FlakeSlide/Materials/M_Pickaxe_FlakeSlide_Ice
graph {
 expressions {
  MaterialExpressionAdd_0 {
   class=MaterialExpressionAdd
   inputs [
    A {from=MaterialExpressionDivide_12}
    B {from=MaterialExpressionNamedRerouteUsage_0}
   ]
  }
  MaterialExpressionAdd_17 {
   class=MaterialExpressionAdd
   inputs [
    A {from=MaterialExpressionDivide_11}
    B {from=MaterialExpressionAppendVector_0}
   ]
   properties {desc=motion}
  }
  MaterialExpressionAdd_18 {
   class=MaterialExpressionAdd
   inputs [
    A {from=MaterialExpressionDivide_8}
    B {from=MaterialExpressionAppendVector_1}
   ]
   properties {desc=motion}
  }
  MaterialExpressionAdd_5 {
   class=MaterialExpressionAdd
   inputs [
    A {from=MaterialExpressionMultiply_26}
    B {from=""}
   ]
  }
  MaterialExpressionAdd_7 {
   class=MaterialExpressionAdd
   inputs [
    A {from=MaterialExpressionTextureCoordinate_3}
    B {from=MaterialExpressionNamedRerouteUsage_57}
   ]
  }
  MaterialExpressionAdd_8 {
   class=MaterialExpressionAdd
   inputs [
    A {from=MaterialExpressionMultiply_38}
    B {from=MaterialExpressionNamedRerouteUsage_88}
   ]
  }
  MaterialExpressionAppendVector_0 {
   class=MaterialExpressionAppendVector
   inputs [
    A {from=MaterialExpressionConstant_11, output=""}
    B {from=MaterialExpressionMultiply_24, output=""}
   ]
  }
  MaterialExpressionAppendVector_1 {
   class=MaterialExpressionAppendVector
   inputs [
    A {from=MaterialExpressionConstant_12, output=""}
    B {from=MaterialExpressionMultiply_23, output=""}
   ]
  }
  MaterialExpressionAppendVector_2 {
   class=MaterialExpressionAppendVector
   inputs [
    A {from=MaterialExpressionVectorParameter_0, output=U Offset}
    B {from=MaterialExpressionVectorParameter_0, output=U Offset}
   ]
  }
  MaterialExpressionBlendMaterialAttributes_0 {
   class=MaterialExpressionBlendMaterialAttributes
   inputs [
    A {from=MaterialExpressionNamedRerouteUsage_4, output=""}
    B {from=MaterialExpressionNamedRerouteUsage_1, output=""}
    Alpha {from=MaterialExpressionTextureSampleParameter2D_3, output=R}
   ]
   properties {}
  }
  MaterialExpressionCameraVectorWS_5 {class=MaterialExpressionCameraVectorWS, inputs [], properties {}}
  MaterialExpressionClamp_1 {
   class=MaterialExpressionClamp
   inputs [
    None {from=MaterialExpressionScalarParameter_25, output=""}
    Min {from="", output=""}
    Max {from="", output=""}
   ]
   properties {maxDefault=5, minDefault=1}
  }
  MaterialExpressionComponentMask_2 {class=MaterialExpressionComponentMask, inputs [None {from=MaterialExpressionNamedRerouteUsage_59}], properties {g=true, r=true}}
  MaterialExpressionComponentMask_4 {class=MaterialExpressionComponentMask, inputs [None {from=MaterialExpressionNamedRerouteUsage_59}], properties {b=true}}
  MaterialExpressionConstantBiasScale_0 {class=MaterialExpressionConstantBiasScale, inputs [None {from=MaterialExpressionNamedRerouteUsage_62}], properties {bias=-0.5, scale=2}}
  MaterialExpressionConstant_0 {class=MaterialExpressionConstant, properties {r=1}}
  MaterialExpressionConstant_10 {class=MaterialExpressionConstant, properties {r=1}}
  MaterialExpressionConstant_11 {class=MaterialExpressionConstant}
  MaterialExpressionConstant_12 {class=MaterialExpressionConstant}
  MaterialExpressionConstant_2 {class=MaterialExpressionConstant}
  MaterialExpressionConstant_3 {class=MaterialExpressionConstant}
  MaterialExpressionConvert_3 {
   class=MaterialExpressionConvert
   inputs [
    None {from=MaterialExpressionDotProduct_1}
    None {from=MaterialExpressionDotProduct_2}
    None {from=MaterialExpressionDotProduct_4}
   ]
   properties {
    nodeName=Make Vector3
    convertInputs [
     {
      type=Scalar
      expressionInput {
       expression=&self:MaterialExpressionDotProduct_1
       inputName=None
       mask=0
       maskA=0
       maskB=0
       maskG=0
       maskR=0
       outputIndex=0
      }
     }
     {
      type=Scalar
      expressionInput {
       expression=&self:MaterialExpressionDotProduct_2
       inputName=None
       mask=0
       maskA=0
       maskB=0
       maskG=0
       maskR=0
       outputIndex=0
      }
     }
     {
      type=Scalar
      expressionInput {
       expression=&self:MaterialExpressionDotProduct_4
       inputName=None
       mask=0
       maskA=0
       maskB=0
       maskG=0
       maskR=0
       outputIndex=0
      }
     }
    ]
    convertOutputs [{type=Vector3}]
   }
  }
  MaterialExpressionDivide_1 {
   class=MaterialExpressionDivide
   inputs [
    A {from=""}
    B {from=MaterialExpressionNamedRerouteUsage_3}
   ]
  }
  MaterialExpressionDivide_11 {
   class=MaterialExpressionDivide
   inputs [
    A {from=MaterialExpressionTextureCoordinate_2}
    B {from=MaterialExpressionNamedRerouteUsage_83}
   ]
  }
  MaterialExpressionDivide_12 {
   class=MaterialExpressionDivide
   inputs [
    A {from=MaterialExpressionTextureCoordinate_0}
    B {from=MaterialExpressionNamedRerouteUsage_156}
   ]
  }
  MaterialExpressionDivide_5 {
   class=MaterialExpressionDivide
   inputs [
    A {from=MaterialExpressionComponentMask_2}
    B {from=MaterialExpressionMax_1}
   ]
  }
  MaterialExpressionDivide_8 {
   class=MaterialExpressionDivide
   inputs [
    A {from=MaterialExpressionAdd_7}
    B {from=MaterialExpressionNamedRerouteUsage_53}
   ]
  }
  MaterialExpressionDivide_9 {
   class=MaterialExpressionDivide
   inputs [
    A {from=MaterialExpressionAdd_8}
    B {from=""}
   ]
   properties {constB=10000}
  }
  MaterialExpressionDotProduct_1 {
   class=MaterialExpressionDotProduct
   inputs [
    A {from=MaterialExpressionNamedRerouteUsage_58, output=""}
    B {from=MaterialExpressionMaterialFunctionCall_28, output=X U}
   ]
   properties {}
  }
  MaterialExpressionDotProduct_2 {
   class=MaterialExpressionDotProduct
   inputs [
    A {from=MaterialExpressionNamedRerouteUsage_58, output=""}
    B {from=MaterialExpressionMaterialFunctionCall_28, output=Y V}
   ]
   properties {}
  }
  MaterialExpressionDotProduct_4 {
   class=MaterialExpressionDotProduct
   inputs [
    A {from=MaterialExpressionNamedRerouteUsage_58, output=""}
    B {from=MaterialExpressionMaterialFunctionCall_28, output=Z W}
   ]
   properties {}
  }
  MaterialExpressionEyeAdaptation_0 {class=MaterialExpressionEyeAdaptation, inputs [], properties {}}
  MaterialExpressionFresnel_0 {
   class=MaterialExpressionFresnel
   inputs [
    ExponentIn {from="", output=""}
    BaseReflectFractionIn {from="", output=""}
    Normal {from="", output=""}
   ]
   properties {baseReflectFraction=0, exponent=1}
  }
  MaterialExpressionLinearInterpolate_0 {
   class=MaterialExpressionLinearInterpolate
   inputs [
    A {from=MaterialExpressionScalarParameter_0}
    B {from=MaterialExpressionScalarParameter_1}
    Alpha {from=MaterialExpressionTextureSampleParameter2D_0, output=R}
   ]
  }
  MaterialExpressionLinearInterpolate_15 {
   class=MaterialExpressionLinearInterpolate
   inputs [
    A {from=MaterialExpressionMaterialFunctionCall_7, output="Distance (S)"}
    B {from=MaterialExpressionMultiply_20}
    Alpha {from=MaterialExpressionNamedRerouteUsage_84}
   ]
  }
  MaterialExpressionMaterialFunctionCall_0 {
   class=MaterialExpressionMaterialFunctionCall
   inputs [In {from=MaterialExpressionStaticSwitchParameter_0}]
   properties {materialFunction=&MF_TechArtOverrides}
  }
  MaterialExpressionMaterialFunctionCall_1 {
   class=MaterialExpressionMaterialFunctionCall
   inputs [
    Sky Illuminance {from=MaterialExpressionNamedRerouteUsage_25}
    Sun Illuminance {from=MaterialExpressionNamedRerouteUsage_26}
    Sun Direction {from=MaterialExpressionNamedRerouteUsage_30}
    Scene Normal {from=MaterialExpressionNamedRerouteUsage_17}
    Scene Albedo {from=MaterialExpressionNamedRerouteUsage_44}
    Scene Color Scale {from=""}
   ]
   properties {
    desc=Scene Luminance Approximation - AlphaBlending
    materialFunction=/Game/Global/MaterialFunctions/TechArt/Refraction/MF_SceneLuminanceFromAlbedo
   }
  }
  MaterialExpressionMaterialFunctionCall_13 {
   class=MaterialExpressionMaterialFunctionCall
   inputs [
    Input {from=MaterialExpressionNamedRerouteUsage_19}
    Input Low {from=MaterialExpressionNamedRerouteUsage_76}
    Input High {from=MaterialExpressionNamedRerouteUsage_77}
   ]
   properties {materialFunction=/Engine/Functions/Engine_MaterialFunctions03/Math/RemapValueRangeNormalized}
  }
  MaterialExpressionMaterialFunctionCall_14 {
   class=MaterialExpressionMaterialFunctionCall
   inputs [
    View Vector {from=MaterialExpressionCameraVectorWS_5}
    Normal {from=MaterialExpressionNamedRerouteUsage_2}
    "Relative IOR (Eta)" {from=MaterialExpressionDivide_1}
   ]
   properties {materialFunction=/Game/Global/MaterialFunctions/TechArt/Refraction/MF_Refract}
  }
  MaterialExpressionMaterialFunctionCall_16 {
   class=MaterialExpressionMaterialFunctionCall
   inputs [
    Normal {from=MaterialExpressionTextureSample_6, output=RGB}
    Intensity {from=MaterialExpressionNamedRerouteUsage_33}
   ]
   properties {materialFunction=/Game/Global/MaterialFunctions/MF_NormalIntensity}
  }
  MaterialExpressionMaterialFunctionCall_17 {
   class=MaterialExpressionMaterialFunctionCall
   inputs [
    Normal_1 {from=MaterialExpressionNamedRerouteUsage_34}
    Normal_2 {from=MaterialExpressionMaterialFunctionCall_16, output="Result (V3)"}
   ]
   properties {materialFunction=/Game/Global/MaterialFunctions/MF_NormalBlend_Whiteout}
  }
  MaterialExpressionMaterialFunctionCall_18 {
   class=MaterialExpressionMaterialFunctionCall
   inputs [
    Normal {from=MaterialExpressionReroute_0}
    Intensity {from=MaterialExpressionNamedRerouteUsage_41}
   ]
   properties {materialFunction=&MF_NormalIntensity}
  }
  MaterialExpressionMaterialFunctionCall_19 {
   class=MaterialExpressionMaterialFunctionCall
   inputs [
    Normal_1 {from=MaterialExpressionMaterialFunctionCall_23, output="Result (V3)"}
    Normal_2 {from=MaterialExpressionNamedRerouteUsage_68}
   ]
   properties {materialFunction=&MF_NormalBlend_Whiteout}
  }
  MaterialExpressionMaterialFunctionCall_2 {
   class=MaterialExpressionMaterialFunctionCall
   inputs [TechArtMask {from=MaterialExpressionTextureSampleParameter2D_4, output=RGB}]
   properties {materialFunction=/Game/Global/Materials/Character_Parent_2023/Glass_Parent_2023/MF_Glass}
  }
  MaterialExpressionMaterialFunctionCall_23 {
   class=MaterialExpressionMaterialFunctionCall
   inputs [
    Normal {from=MaterialExpressionNamedRerouteUsage_36}
    Intensity {from=MaterialExpressionNamedRerouteUsage_145}
   ]
   properties {materialFunction=&MF_NormalIntensity}
  }
  MaterialExpressionMaterialFunctionCall_26 {
   class=MaterialExpressionMaterialFunctionCall
   inputs [
    Input {from=MaterialExpressionMaterialFunctionCall_7, output="Distance (S)"}
    Input Low {from=MaterialExpressionNamedRerouteUsage_93}
    Input High {from=MaterialExpressionNamedRerouteUsage_94}
   ]
   properties {materialFunction=&RemapValueRangeNormalized}
  }
  MaterialExpressionMaterialFunctionCall_27 {
   class=MaterialExpressionMaterialFunctionCall
   inputs [
    Input {from=MaterialExpressionNamedRerouteUsage_61}
    Input Low {from=MaterialExpressionNamedRerouteUsage_54}
    Input High {from=MaterialExpressionNamedRerouteUsage_55}
   ]
   properties {materialFunction=&RemapValueRangeNormalized}
  }
  MaterialExpressionMaterialFunctionCall_28 {
   class=MaterialExpressionMaterialFunctionCall
   inputs []
   properties {materialFunction=&TangentBasis}
  }
  MaterialExpressionMaterialFunctionCall_3 {
   class=MaterialExpressionMaterialFunctionCall
   inputs [
    Material Attributes {from=MaterialExpressionMaterialFunctionCall_2, output=Result}
    Fresnel {from=MaterialExpressionFresnel_0}
    Alpha {from=MaterialExpressionReroute_1}
   ]
   properties {materialFunction=/Game/Global/MaterialFunctions/TechArt/MF_ThinFilm_TechArt}
  }
  MaterialExpressionMaterialFunctionCall_30 {
   class=MaterialExpressionMaterialFunctionCall
   inputs [In {from=MaterialExpressionSetMaterialAttributes_5}]
   properties {materialFunction=&MF_TechArtOverrides}
  }
  MaterialExpressionMaterialFunctionCall_35 {
   class=MaterialExpressionMaterialFunctionCall
   inputs [
    Normal WS {from=MaterialExpressionNamedRerouteUsage_40}
    Roughness {from=MaterialExpressionNamedRerouteUsage_79}
    Surface Opacity {from=MaterialExpressionNamedRerouteUsage_20}
    Refraction Vector WS {from=MaterialExpressionNamedRerouteUsage_28}
    Refraction IOR {from=MaterialExpressionNamedRerouteUsage_38}
    Scene Luminance {from=MaterialExpressionNamedRerouteUsage_11}
    Interior Emission {from=MaterialExpressionConstant_2}
    Volume Thickness {from=MaterialExpressionNamedRerouteUsage_12}
    Volume Density {from=MaterialExpressionNamedRerouteUsage_63}
    Absorption Coeffs {from=MaterialExpressionNamedRerouteUsage_15}
    Scattering Coeffs {from=MaterialExpressionNamedRerouteUsage_16}
    Scattering Phase {from=MaterialExpressionNamedRerouteUsage_70}
    Is Scene Inside Volume {from=MaterialExpressionConstant_10}
    Thin Transmission Exp {from=MaterialExpressionNamedRerouteUsage_64}
    //---- ALPHA-BLENDING ONLY ----// {from=""}
    Haze Amount {from=MaterialExpressionNamedRerouteUsage_9}
    //----------- LIGHTING -----------// {from=""}
    Sky Illuminance {from=MaterialExpressionNamedRerouteUsage_35}
    Sun Illuminance {from=MaterialExpressionNamedRerouteUsage_37}
    Sun Direction {from=MaterialExpressionNamedRerouteUsage_39}
   ]
   properties {materialFunction=/Game/Global/MaterialFunctions/TechArt/Refraction/MF_DielectricInteriorVolume}
  }
  MaterialExpressionMaterialFunctionCall_4 {
   class=MaterialExpressionMaterialFunctionCall
   inputs [
    "Roughness^2" {from=MaterialExpressionMultiply_2}
    Roughness Exponent {from=MaterialExpressionNamedRerouteUsage_50}
    Haze Min Amount {from=MaterialExpressionNamedRerouteUsage_48}
    Haze Max Amount {from=MaterialExpressionNamedRerouteUsage_47}
   ]
   properties {materialFunction=/Game/Global/MaterialFunctions/TechArt/Refraction/MF_VolumeHazeAmount}
  }
  MaterialExpressionMaterialFunctionCall_5 {
   class=MaterialExpressionMaterialFunctionCall
   inputs [
    Normal WS {from=MaterialExpressionNamedRerouteUsage_13}
    Sky Illuminance Mult {from=""}
    Sky Desaturation {from=""}
    Sun Illuminance Mult {from=""}
    Sun Desaturation {from=""}
   ]
   properties {materialFunction=/Game/Global/MaterialFunctions/TechArt/Refraction/MF_SunSkyProperties}
  }
  MaterialExpressionMaterialFunctionCall_6 {
   class=MaterialExpressionMaterialFunctionCall
   inputs [In {from=MaterialExpressionMaterialFunctionCall_3, output=Result}]
   properties {materialFunction=/Game/Packages/Fortress_SharedMaterials/Base_Material_Functions/BaseCustomizationFunctions/MF_FN_Vehicle_CameraDither}
  }
  MaterialExpressionMaterialFunctionCall_7 {
   class=MaterialExpressionMaterialFunctionCall
   inputs [
    Viewport UV {from=MaterialExpressionNamedRerouteUsage_23}
    Max Distance {from=""}
   ]
   properties {materialFunction=/Game/Global/MaterialFunctions/TechArt/Refraction/MF_DistanceToRefractedScene}
  }
  MaterialExpressionMaterialFunctionCall_8 {
   class=MaterialExpressionMaterialFunctionCall
   inputs [
    Normal TS {from=MaterialExpressionNamedRerouteUsage_7}
    Refraction IOR {from=MaterialExpressionNamedRerouteUsage_5}
    Distortion Scale {from=""}
    Depth Contribution {from=""}
    Pixel Normal Offset Threshold {from=""}
    Enable Refraction Mask {from=""}
   ]
   properties {materialFunction=/Game/Global/MaterialFunctions/TechArt/Refraction/MF_RefractionDistortion_ScreenSpace}
  }
  MaterialExpressionMaterialFunctionCall_9 {
   class=MaterialExpressionMaterialFunctionCall
   inputs [Material {from=MaterialExpressionStaticSwitchParameter_3}]
   properties {materialFunction=/Game/Global/Materials/Character_Parent_2023/Functions/BaseFN_CharacterOpacity_2023}
  }
  MaterialExpressionMax_0 {
   class=MaterialExpressionMax
   inputs [
    A {from=MaterialExpressionMultiply_15}
    B {from=""}
   ]
   properties {constB=0}
  }
  MaterialExpressionMax_1 {
   class=MaterialExpressionMax
   inputs [
    A {from=MaterialExpressionComponentMask_4}
    B {from=""}
   ]
   properties {constB=0.001}
  }
  MaterialExpressionMultiply_0 {
   class=MaterialExpressionMultiply
   inputs [
    A {from=MaterialExpressionLinearInterpolate_0}
    B {from=""}
   ]
  }
  MaterialExpressionMultiply_12 {
   class=MaterialExpressionMultiply
   inputs [
    A {from=MaterialExpressionMultiply_29}
    B {from=MaterialExpressionNamedRerouteUsage_42}
   ]
   properties {desc=Scene Luminance - Opaque}
  }
  MaterialExpressionMultiply_15 {
   class=MaterialExpressionMultiply
   inputs [
    A {from=MaterialExpressionVectorParameter_6, output=RGB}
    B {from=MaterialExpressionVectorParameter_6, output=RGB}
   ]
  }
  MaterialExpressionMultiply_2 {
   class=MaterialExpressionMultiply
   inputs [
    A {from=MaterialExpressionNamedRerouteUsage_8}
    B {from=MaterialExpressionNamedRerouteUsage_8}
   ]
  }
  MaterialExpressionMultiply_20 {
   class=MaterialExpressionMultiply
   inputs [
    A {from=MaterialExpressionPower_2}
    B {from=MaterialExpressionReroute_4}
   ]
  }
  MaterialExpressionMultiply_21 {
   class=MaterialExpressionMultiply
   inputs [
    A {from=MaterialExpressionTime_0}
    B {from=MaterialExpressionNamedRerouteUsage_113}
   ]
  }
  MaterialExpressionMultiply_22 {
   class=MaterialExpressionMultiply
   inputs [
    A {from=MaterialExpressionMultiply_21}
    B {from=""}
   ]
   properties {constB=0.004}
  }
  MaterialExpressionMultiply_23 {
   class=MaterialExpressionMultiply
   inputs [
    A {from=MaterialExpressionNamedRerouteUsage_51}
    B {from=""}
   ]
   properties {constB=-0.75}
  }
  MaterialExpressionMultiply_24 {
   class=MaterialExpressionMultiply
   inputs [
    A {from=MaterialExpressionMultiply_22}
    B {from=""}
   ]
   properties {constB=0.5}
  }
  MaterialExpressionMultiply_25 {
   class=MaterialExpressionMultiply
   inputs [
    A {from=MaterialExpressionNamedRerouteUsage_80}
    B {from=MaterialExpressionVectorParameter_8, output=Breakup Pattern Relative Scale}
   ]
  }
  MaterialExpressionMultiply_26 {
   class=MaterialExpressionMultiply
   inputs [
    A {from=MaterialExpressionConstantBiasScale_0}
    B {from=MaterialExpressionNamedRerouteUsage_82}
   ]
  }
  MaterialExpressionMultiply_27 {
   class=MaterialExpressionMultiply
   inputs [
    A {from=MaterialExpressionAdd_5}
    B {from=MaterialExpressionNamedRerouteUsage_14}
   ]
  }
  MaterialExpressionMultiply_29 {
   class=MaterialExpressionMultiply
   inputs [
    A {from=MaterialExpressionSceneColor_0, output=RGB}
    B {from=MaterialExpressionEyeAdaptation_0, output=EyeAdaptation}
   ]
  }
  MaterialExpressionMultiply_3 {
   class=MaterialExpressionMultiply
   inputs [
    A {from=MaterialExpressionMaterialFunctionCall_4, output="Haze Amount (S)"}
    B {from=MaterialExpressionNamedRerouteUsage_46}
   ]
  }
  MaterialExpressionMultiply_30 {
   class=MaterialExpressionMultiply
   inputs [
    A {from=MaterialExpressionVectorParameter_3, output=RGB}
    B {from=MaterialExpressionVectorParameter_3, output=RGB}
   ]
  }
  MaterialExpressionMultiply_35 {
   class=MaterialExpressionMultiply
   inputs [
    A {from=MaterialExpressionDivide_9}
    B {from=MaterialExpressionNamedRerouteUsage_60}
   ]
   properties {desc=UV Parallax Offset}
  }
  MaterialExpressionMultiply_38 {
   class=MaterialExpressionMultiply
   inputs [
    A {from=MaterialExpressionPower_6}
    B {from=MaterialExpressionNamedRerouteUsage_87}
   ]
  }
  MaterialExpressionMultiply_4 {
   class=MaterialExpressionMultiply
   inputs [
    A {from=MaterialExpressionVectorParameter_1, output=RGB}
    B {from=MaterialExpressionVectorParameter_1, output=RGB}
   ]
  }
  MaterialExpressionMultiply_5 {
   class=MaterialExpressionMultiply
   inputs [
    A {from=MaterialExpressionSubtract_0}
    B {from=MaterialExpressionScalarParameter_2}
   ]
  }
  MaterialExpressionMultiply_6 {
   class=MaterialExpressionMultiply
   inputs [
    A {from=MaterialExpressionNamedRerouteUsage_27}
    B {from=MaterialExpressionNamedRerouteUsage_24}
   ]
  }
  MaterialExpressionMultiply_7 {
   class=MaterialExpressionMultiply
   inputs [
    A {from=MaterialExpressionPower_4}
    B {from=MaterialExpressionNamedRerouteUsage_32}
   ]
  }
  MaterialExpressionNamedRerouteDeclaration_0 {
   class=MaterialExpressionNamedRerouteDeclaration
   inputs [None {from=MaterialExpressionAppendVector_2}]
   properties {name="Surface UV Offset (V2)", nodeColor=<rgba 1 0.2578 0 1>}
  }
  MaterialExpressionNamedRerouteDeclaration_1 {
   class=MaterialExpressionNamedRerouteDeclaration
   inputs [None {from=MaterialExpressionTransform_2}]
   properties {name="Refraction Normal WS (V3)", nodeColor=<rgba 0 0.5913 1 1>}
  }
  MaterialExpressionNamedRerouteDeclaration_10 {
   class=MaterialExpressionNamedRerouteDeclaration
   inputs [None {from=MaterialExpressionMax_0}]
   properties {name="Refracted Scene Intensity (V3)", nodeColor=<rgba 0.7578 1 0 1>}
  }
  MaterialExpressionNamedRerouteDeclaration_100 {
   class=MaterialExpressionNamedRerouteDeclaration
   inputs [None {from=MaterialExpressionVectorParameter_0, output=UV Scale}]
   properties {name="Surface UV Scale (S)", nodeColor=<rgba 0 1 0.1328 1>}
  }
  MaterialExpressionNamedRerouteDeclaration_102 {
   class=MaterialExpressionNamedRerouteDeclaration
   inputs [None {from=MaterialExpressionMultiply_4}]
   properties {name="Absorption Coeffs In (V3)", nodeColor=<rgba 1 0 0.5391 1>}
  }
  MaterialExpressionNamedRerouteDeclaration_103 {
   class=MaterialExpressionNamedRerouteDeclaration
   inputs [None {from=MaterialExpressionMultiply_30}]
   properties {name="Scattering Coeffs In (V3)", nodeColor=<rgba 0 1 0.4375 1>}
  }
  MaterialExpressionNamedRerouteDeclaration_107 {
   class=MaterialExpressionNamedRerouteDeclaration
   inputs [None {from=MaterialExpressionTextureObjectParameter_3}]
   properties {name="Surface Roughness Map (TO)", nodeColor=<rgba 0.125 1 0 1>}
  }
  MaterialExpressionNamedRerouteDeclaration_108 {
   class=MaterialExpressionNamedRerouteDeclaration
   inputs [None {from=MaterialExpressionTextureObjectParameter_4}]
   properties {name="Surface Detail Normal Map (TO)", nodeColor=<rgba 0.6406 1 0 1>}
  }
  MaterialExpressionNamedRerouteDeclaration_11 {
   class=MaterialExpressionNamedRerouteDeclaration
   inputs [None {from=MaterialExpressionShadingPathSwitch_2}]
   properties {name="Scene Luminance (V3)", nodeColor=<rgba 0 0.6719 1 1>}
  }
  MaterialExpressionNamedRerouteDeclaration_115 {
   class=MaterialExpressionNamedRerouteDeclaration
   inputs [None {from=MaterialExpressionVectorParameter_8, output=Breakup Pattern Mult}]
   properties {name="Noise Depth Variation Mult (S)", nodeColor=<rgba 0 0.0156 1 1>}
  }
  MaterialExpressionNamedRerouteDeclaration_116 {
   class=MaterialExpressionNamedRerouteDeclaration
   inputs [None {from=MaterialExpressionVectorParameter_8, output=Add Constant Depth}]
   properties {name="Noise Add Constant Depth (S)", nodeColor=<rgba 0.3359 1 0 1>}
  }
  MaterialExpressionNamedRerouteDeclaration_12 {
   class=MaterialExpressionNamedRerouteDeclaration
   inputs [None {from=MaterialExpressionMultiply_3}]
   properties {name="Haze Amount (S)", desc=ALPHA-BLENDING ONLY, nodeColor=<rgba 0.1016 0 1 1>}
  }
  MaterialExpressionNamedRerouteDeclaration_121 {
   class=MaterialExpressionNamedRerouteDeclaration
   inputs [None {from=MaterialExpressionScalarParameter_48}]
   properties {name="Interior Noise Speed (S)", nodeColor=<rgba 0 1 0.0625 1>}
  }
  MaterialExpressionNamedRerouteDeclaration_13 {
   class=MaterialExpressionNamedRerouteDeclaration
   inputs [None {from=MaterialExpressionMaterialFunctionCall_30, output=Result}]
   properties {name=CubeIce, nodeColor=<rgba 1 0.8906 0 1>}
  }
  MaterialExpressionNamedRerouteDeclaration_14 {
   class=MaterialExpressionNamedRerouteDeclaration
   inputs [None {from=MaterialExpressionNamedRerouteUsage_158}]
   properties {name="Volume Absorption Coeffs (V3)", nodeColor=<rgba 0 0.2031 1 1>}
  }
  MaterialExpressionNamedRerouteDeclaration_15 {
   class=MaterialExpressionNamedRerouteDeclaration
   inputs [None {from=MaterialExpressionMultiply_27}]
   properties {name="Volume Scattering Coeffs (V3)", nodeColor=<rgba 0.5234 1 0 1>}
  }
  MaterialExpressionNamedRerouteDeclaration_16 {
   class=MaterialExpressionNamedRerouteDeclaration
   inputs [None {from=MaterialExpressionShadingPathSwitch_4}]
   properties {name="Material Opacity (S)", nodeColor=<rgba 0.2422 0 1 1>}
  }
  MaterialExpressionNamedRerouteDeclaration_17 {
   class=MaterialExpressionNamedRerouteDeclaration
   inputs [None {from=MaterialExpressionTextureSampleParameter2D_1, output=A}]
   properties {name="Opacity Map (S)", nodeColor=<rgba 0 1 0.8359 1>}
  }
  MaterialExpressionNamedRerouteDeclaration_18 {
   class=MaterialExpressionNamedRerouteDeclaration
   inputs [None {from=MaterialExpressionPower_0}]
   properties {name="Surface Opacity (S)", nodeColor=<rgba 1 0.5156 0 1>}
  }
  MaterialExpressionNamedRerouteDeclaration_19 {
   class=MaterialExpressionNamedRerouteDeclaration
   inputs [None {from=MaterialExpressionMaterialFunctionCall_5, output="Sun Direction (V3)"}]
   properties {name="Lighting - Sun Direction (V3)", nodeColor=<rgba 0 0.0625 1 1>}
  }
  MaterialExpressionNamedRerouteDeclaration_2 {
   class=MaterialExpressionNamedRerouteDeclaration
   inputs [None {from=MaterialExpressionMaterialFunctionCall_14, output="Refraction Vector (V3)"}]
   properties {name="Refraction Vector WS (V3)", nodeColor=<rgba 0.5469 1 0 1>}
  }
  MaterialExpressionNamedRerouteDeclaration_20 {
   class=MaterialExpressionNamedRerouteDeclaration
   inputs [None {from=MaterialExpressionVectorParameter_10, output=Specular}]
   properties {name="Surface Specular (S)", nodeColor=<rgba 0.1953 0 1 1>}
  }
  MaterialExpressionNamedRerouteDeclaration_21 {
   class=MaterialExpressionNamedRerouteDeclaration
   inputs [None {from=MaterialExpressionSaturate_10}]
   properties {name="Surface Roughness (S)", nodeColor=<rgba 0.125 1 0 1>}
  }
  MaterialExpressionNamedRerouteDeclaration_22 {
   class=MaterialExpressionNamedRerouteDeclaration
   inputs [None {from=MaterialExpressionTextureSampleParameter2D_1, output=RGB}]
   properties {name="Diffuse Color (V3)", nodeColor=<rgba 1 0 0.4453 1>}
  }
  MaterialExpressionNamedRerouteDeclaration_23 {
   class=MaterialExpressionNamedRerouteDeclaration
   inputs [None {from=MaterialExpressionTextureSampleParameter2D_2, output=RGB}]
   properties {name="Surface Normal In TS (V3)", nodeColor=<rgba 0 0.7656 1 1>}
  }
  MaterialExpressionNamedRerouteDeclaration_24 {
   class=MaterialExpressionNamedRerouteDeclaration
   inputs [None {from=MaterialExpressionMaterialFunctionCall_5, output="Sun Illuminance (V3)"}]
   properties {name="Lighting - Sun Illuminance (V3)", nodeColor=<rgba 0.7344 0 1 1>}
  }
  MaterialExpressionNamedRerouteDeclaration_25 {
   class=MaterialExpressionNamedRerouteDeclaration
   inputs [None {from=MaterialExpressionPower_8}]
   properties {name="Refraction Depth Variation (S)", nodeColor=<rgba 1 0 0 1>}
  }
  MaterialExpressionNamedRerouteDeclaration_26 {
   class=MaterialExpressionNamedRerouteDeclaration
   inputs [None {from=MaterialExpressionMaterialFunctionCall_18, output="Result (V3)"}]
   properties {name="Refraction Detail Normal TS (V3)", nodeColor=<rgba 0 1 0.5781 1>}
  }
  MaterialExpressionNamedRerouteDeclaration_28 {
   class=MaterialExpressionNamedRerouteDeclaration
   inputs [None {from=MaterialExpressionVectorParameter_7, output=Pattern Scale}]
   properties {name="Interior Noise Scale (S)", nodeColor=<rgba 1 0 0.0938 1>}
  }
  MaterialExpressionNamedRerouteDeclaration_29 {
   class=MaterialExpressionNamedRerouteDeclaration
   inputs [None {from=MaterialExpressionConvert_3}]
   properties {name="Refraction Vector TS (V3)", nodeColor=<rgba 0 1 0.3203 1>}
  }
  MaterialExpressionNamedRerouteDeclaration_3 {
   class=MaterialExpressionNamedRerouteDeclaration
   inputs [None {from=MaterialExpressionMaterialFunctionCall_19, output=Result}]
   properties {name="Refraction Normal TS (V3)", nodeColor=<rgba 0 0.5913 1 1>}
  }
  MaterialExpressionNamedRerouteDeclaration_30 {
   class=MaterialExpressionNamedRerouteDeclaration
   inputs [None {from=MaterialExpressionDivide_5}]
   properties {name="Refracted Unit UV Offset TS (V2)", nodeColor=<rgba 1 0 0.1875 1>}
  }
  MaterialExpressionNamedRerouteDeclaration_31 {
   class=MaterialExpressionNamedRerouteDeclaration
   inputs [None {from=MaterialExpressionMaterialFunctionCall_5, output="Sky Illuminance (V3)"}]
   properties {name="Lighting - Sky Illuminance (V3)", nodeColor=<rgba 0 1 0.4141 1>}
  }
  MaterialExpressionNamedRerouteDeclaration_32 {
   class=MaterialExpressionNamedRerouteDeclaration
   inputs [None {from=MaterialExpressionSaturate_4}]
   properties {name="Refracted Scene Avg Color (V3)", nodeColor=<rgba 0 0.625 1 1>}
  }
  MaterialExpressionNamedRerouteDeclaration_33 {
   class=MaterialExpressionNamedRerouteDeclaration
   inputs [None {from=MaterialExpressionAdd_0}]
   properties {name="Surface UV (V2)", nodeColor=<rgba 1 0 0.2109 1>}
  }
  MaterialExpressionNamedRerouteDeclaration_34 {
   class=MaterialExpressionNamedRerouteDeclaration
   inputs [None {from=MaterialExpressionVectorParameter_13, output=Roughness Exponent}]
   properties {name="Haze Roughness Exponent (S)", nodeColor=<rgba 0.9453 1 0 1>}
  }
  MaterialExpressionNamedRerouteDeclaration_35 {
   class=MaterialExpressionNamedRerouteDeclaration
   inputs [None {from=MaterialExpressionVectorParameter_13, output=Haze Min Amount}]
   properties {name="Haze Min Amount (S)", nodeColor=<rgba 0.7344 0 1 1>}
  }
  MaterialExpressionNamedRerouteDeclaration_37 {
   class=MaterialExpressionNamedRerouteDeclaration
   inputs [None {from=MaterialExpressionVectorParameter_9, output=Opacity Old Min}]
   properties {name="Surface Opacity OldMin (S)", nodeColor=<rgba 0 1 0.625 1>}
  }
  MaterialExpressionNamedRerouteDeclaration_38 {
   class=MaterialExpressionNamedRerouteDeclaration
   inputs [None {from=MaterialExpressionVectorParameter_9, output=Opacity Old Max}]
   properties {name="Surface Opacity Old Max (S)", nodeColor=<rgba 1 0.3047 0 1>}
  }
  MaterialExpressionNamedRerouteDeclaration_39 {
   class=MaterialExpressionNamedRerouteDeclaration
   inputs [None {from=MaterialExpressionVectorParameter_9, output=Opacity Exp}]
   properties {name="Surface Opacity Exp (S)", nodeColor=<rgba 0 0.0156 1 1>}
  }
  MaterialExpressionNamedRerouteDeclaration_4 {
   class=MaterialExpressionNamedRerouteDeclaration
   inputs [None {from=MaterialExpressionLinearInterpolate_15}]
   properties {name="Thickness (S)", nodeColor=<rgba 0 1 0.9297 1>}
  }
  MaterialExpressionNamedRerouteDeclaration_40 {
   class=MaterialExpressionNamedRerouteDeclaration
   inputs [None {from=MaterialExpressionScalarParameter_49}]
   properties {name="Interior Noise Amount (S)", nodeColor=<rgba 0.875 0 1 1>}
  }
  MaterialExpressionNamedRerouteDeclaration_41 {
   class=MaterialExpressionNamedRerouteDeclaration
   inputs [None {from=MaterialExpressionVectorParameter_11, output=Thickness Remap Mix}]
   properties {name="Thickness Remap Mix (S)", nodeColor=<rgba 0 1 0.5547 1>}
  }
  MaterialExpressionNamedRerouteDeclaration_42 {
   class=MaterialExpressionNamedRerouteDeclaration
   inputs [None {from=MaterialExpressionVectorParameter_11, output=Thickness OldMin}]
   properties {name="Thickness Old Min (S)", nodeColor=<rgba 1 0.2344 0 1>}
  }
  MaterialExpressionNamedRerouteDeclaration_43 {
   class=MaterialExpressionNamedRerouteDeclaration
   inputs [None {from=MaterialExpressionVectorParameter_11, output=Thickness OldMax}]
   properties {name="Thickness Old Max (S)", nodeColor=<rgba 0 0.0859 1 1>}
  }
  MaterialExpressionNamedRerouteDeclaration_44 {
   class=MaterialExpressionNamedRerouteDeclaration
   inputs [None {from=MaterialExpressionVectorParameter_11, output=Thickness Exp}]
   properties {name="Thickness Exp (S)", nodeColor=<rgba 0.4062 1 0 1>}
  }
  MaterialExpressionNamedRerouteDeclaration_45 {
   class=MaterialExpressionNamedRerouteDeclaration
   inputs [None {from=MaterialExpressionMultiply_6}]
   properties {name="Surface Color (V3)", nodeColor=<rgba 1 0 0.7266 1>}
  }
  MaterialExpressionNamedRerouteDeclaration_46 {
   class=MaterialExpressionNamedRerouteDeclaration
   inputs [None {from=MaterialExpressionVectorParameter_13, output=Haze Max Amount}]
   properties {name="Haze Max Amount (S)", nodeColor=<rgba 0 1 0.4141 1>}
  }
  MaterialExpressionNamedRerouteDeclaration_47 {
   class=MaterialExpressionNamedRerouteDeclaration
   inputs [None {from=MaterialExpressionScalarParameter_76}]
   properties {name="Volume Density (S)", nodeColor=<rgba 0 1 0.9531 1>}
  }
  MaterialExpressionNamedRerouteDeclaration_48 {
   class=MaterialExpressionNamedRerouteDeclaration
   inputs [None {from=MaterialExpressionScalarParameter_78}]
   properties {name="Transmission Attenuation (S)", nodeColor=<rgba 1 0.6328 0 1>}
  }
  MaterialExpressionNamedRerouteDeclaration_49 {
   class=MaterialExpressionNamedRerouteDeclaration
   inputs [None {from=MaterialExpressionScalarParameter_77}]
   properties {name="Volume Scattering Phase (S)", nodeColor=<rgba 0.3125 0 1 1>}
  }
  MaterialExpressionNamedRerouteDeclaration_5 {
   class=MaterialExpressionNamedRerouteDeclaration
   inputs [None {from=MaterialExpressionMaterialFunctionCall_8, output="Refracted Viewport UV (V2)"}]
   properties {name="Refracted Viewport UV (V2)", nodeColor=<rgba 0.9219 1 0 1>}
  }
  MaterialExpressionNamedRerouteDeclaration_50 {
   class=MaterialExpressionNamedRerouteDeclaration
   inputs [None {from=MaterialExpressionSaturate_0}]
   properties {name="Enable Haze (S)", nodeColor=<rgba 1 0.0938 0 1>}
  }
  MaterialExpressionNamedRerouteDeclaration_51 {
   class=MaterialExpressionNamedRerouteDeclaration
   inputs [None {from=MaterialExpressionTextureSample_5, output=R}]
   properties {name="Depth Breakup Pattern (S)", nodeColor=<rgba 0.0078 1 0 1>}
  }
  MaterialExpressionNamedRerouteDeclaration_52 {
   class=MaterialExpressionNamedRerouteDeclaration
   inputs [None {from=MaterialExpressionMultiply_35}]
   properties {name="Noise UV Parallax Offset (V2)", nodeColor=<rgba 1 0 0.3281 1>}
  }
  MaterialExpressionNamedRerouteDeclaration_53 {
   class=MaterialExpressionNamedRerouteDeclaration
   inputs [None {from=MaterialExpressionTextureSample_15, output=R}]
   properties {name="Parallax Pattern (S)", nodeColor=<rgba 0 0.6484 1 1>}
  }
  MaterialExpressionNamedRerouteDeclaration_57 {
   class=MaterialExpressionNamedRerouteDeclaration
   inputs [None {from=MaterialExpressionVectorParameter_7, output=Pattern OldMin}]
   properties {name="Interior Noise OldMin (S)", nodeColor=<rgba 0 0.4141 1 1>}
  }
  MaterialExpressionNamedRerouteDeclaration_59 {
   class=MaterialExpressionNamedRerouteDeclaration
   inputs [None {from=MaterialExpressionVectorParameter_7, output=Pattern OldMax}]
   properties {name="Interior Noise OldMax (S)", nodeColor=<rgba 0.7344 1 0 1>}
  }
  MaterialExpressionNamedRerouteDeclaration_6 {
   class=MaterialExpressionNamedRerouteDeclaration
   inputs [None {from=MaterialExpressionScalarParameter_38}]
   properties {name="Surface Detail Normal Intensity (S)", nodeColor=<rgba 0 1 0.6953 1>}
  }
  MaterialExpressionNamedRerouteDeclaration_63 {
   class=MaterialExpressionNamedRerouteDeclaration
   inputs [None {from=MaterialExpressionMultiply_22}]
   properties {name="Time (S)", nodeColor=<rgba 0 0.7656 1 1>}
  }
  MaterialExpressionNamedRerouteDeclaration_7 {
   class=MaterialExpressionNamedRerouteDeclaration
   inputs [None {from=MaterialExpressionClamp_1}]
   properties {name="Refraction - IOR (S)", nodeColor=<rgba 1 0.0938 0 1>}
  }
  MaterialExpressionNamedRerouteDeclaration_8 {
   class=MaterialExpressionNamedRerouteDeclaration
   inputs [None {from=MaterialExpressionMaterialFunctionCall_17, output=Result}]
   properties {name="Surface Normal TS (V3)", nodeColor=<rgba 1 0.375 0 1>}
  }
  MaterialExpressionNamedRerouteDeclaration_80 {
   class=MaterialExpressionNamedRerouteDeclaration
   inputs [None {from=MaterialExpressionVectorParameter_7, output=Pattern Exp}]
   properties {name="Interior Noise Exp (S)", nodeColor=<rgba 0.9453 0 1 1>}
  }
  MaterialExpressionNamedRerouteDeclaration_83 {
   class=MaterialExpressionNamedRerouteDeclaration
   inputs [None {from=MaterialExpressionMultiply_25}]
   properties {name="Noise Depth Variation Scale (S)", nodeColor=<rgba 0 1 0.625 1>}
  }
  MaterialExpressionNamedRerouteDeclaration_84 {
   class=MaterialExpressionNamedRerouteDeclaration
   inputs [None {from=MaterialExpressionVectorParameter_8, output=Breakup Pattern Exp}]
   properties {name="Noise Depth Variation Exp (S)", nodeColor=<rgba 1 0.3047 0 1>}
  }
  MaterialExpressionNamedRerouteDeclaration_9 {
   class=MaterialExpressionNamedRerouteDeclaration
   inputs [None {from=MaterialExpressionMaterialFunctionCall_0, output=Result}]
   properties {name=TailIce, nodeColor=<rgba 0 0.7891 1 1>}
  }
  MaterialExpressionNamedRerouteDeclaration_90 {
   class=MaterialExpressionNamedRerouteDeclaration
   inputs [None {from=MaterialExpressionScalarParameter_39}]
   properties {name="Refraction Detail Normal Mix (S)", nodeColor=<rgba 1 0 0.8906 1>}
  }
  MaterialExpressionNamedRerouteDeclaration_92 {
   class=MaterialExpressionNamedRerouteDeclaration
   inputs [None {from=MaterialExpressionScalarParameter_27}]
   properties {name="Refraction Surface Normal Mix (S)", nodeColor=<rgba 1 0.8203 0 1>}
  }
  MaterialExpressionNamedRerouteDeclaration_95 {
   class=MaterialExpressionNamedRerouteDeclaration
   inputs [None {from=MaterialExpressionShadingPathSwitch_3}]
   properties {name="Refraction Result (V3)", nodeColor=<rgba 1 0 0.75 1>}
  }
  MaterialExpressionNamedRerouteDeclaration_98 {
   class=MaterialExpressionNamedRerouteDeclaration
   inputs [None {from=MaterialExpressionVectorParameter_10, output=Roughness Exp}]
   properties {name="Surface Roughness Exp (S)", nodeColor=<rgba 1 0.9844 0 1>}
  }
  MaterialExpressionNamedRerouteDeclaration_99 {
   class=MaterialExpressionNamedRerouteDeclaration
   inputs [None {from=MaterialExpressionVectorParameter_10, output=Roughness Mult}]
   properties {name="Surface Roughness Mult (S)", nodeColor=<rgba 0.6641 0 1 1>}
  }
  MaterialExpressionNamedRerouteUsage_0 {class=MaterialExpressionNamedRerouteUsage}
  MaterialExpressionNamedRerouteUsage_1 {class=MaterialExpressionNamedRerouteUsage}
  MaterialExpressionNamedRerouteUsage_10 {class=MaterialExpressionNamedRerouteUsage}
  MaterialExpressionNamedRerouteUsage_11 {class=MaterialExpressionNamedRerouteUsage}
  MaterialExpressionNamedRerouteUsage_113 {class=MaterialExpressionNamedRerouteUsage}
  MaterialExpressionNamedRerouteUsage_119 {class=MaterialExpressionNamedRerouteUsage}
  MaterialExpressionNamedRerouteUsage_12 {class=MaterialExpressionNamedRerouteUsage}
  MaterialExpressionNamedRerouteUsage_121 {class=MaterialExpressionNamedRerouteUsage}
  MaterialExpressionNamedRerouteUsage_13 {class=MaterialExpressionNamedRerouteUsage}
  MaterialExpressionNamedRerouteUsage_14 {class=MaterialExpressionNamedRerouteUsage}
  MaterialExpressionNamedRerouteUsage_145 {class=MaterialExpressionNamedRerouteUsage}
  MaterialExpressionNamedRerouteUsage_15 {class=MaterialExpressionNamedRerouteUsage}
  MaterialExpressionNamedRerouteUsage_156 {class=MaterialExpressionNamedRerouteUsage}
  MaterialExpressionNamedRerouteUsage_158 {class=MaterialExpressionNamedRerouteUsage}
  MaterialExpressionNamedRerouteUsage_16 {class=MaterialExpressionNamedRerouteUsage}
  MaterialExpressionNamedRerouteUsage_164 {class=MaterialExpressionNamedRerouteUsage}
  MaterialExpressionNamedRerouteUsage_165 {class=MaterialExpressionNamedRerouteUsage}
  MaterialExpressionNamedRerouteUsage_17 {class=MaterialExpressionNamedRerouteUsage}
  MaterialExpressionNamedRerouteUsage_18 {class=MaterialExpressionNamedRerouteUsage}
  MaterialExpressionNamedRerouteUsage_19 {class=MaterialExpressionNamedRerouteUsage}
  MaterialExpressionNamedRerouteUsage_2 {class=MaterialExpressionNamedRerouteUsage}
  MaterialExpressionNamedRerouteUsage_20 {class=MaterialExpressionNamedRerouteUsage}
  MaterialExpressionNamedRerouteUsage_21 {class=MaterialExpressionNamedRerouteUsage}
  MaterialExpressionNamedRerouteUsage_22 {class=MaterialExpressionNamedRerouteUsage}
  MaterialExpressionNamedRerouteUsage_23 {class=MaterialExpressionNamedRerouteUsage}
  MaterialExpressionNamedRerouteUsage_24 {class=MaterialExpressionNamedRerouteUsage}
  MaterialExpressionNamedRerouteUsage_25 {class=MaterialExpressionNamedRerouteUsage}
  MaterialExpressionNamedRerouteUsage_26 {class=MaterialExpressionNamedRerouteUsage}
  MaterialExpressionNamedRerouteUsage_27 {class=MaterialExpressionNamedRerouteUsage}
  MaterialExpressionNamedRerouteUsage_28 {class=MaterialExpressionNamedRerouteUsage}
  MaterialExpressionNamedRerouteUsage_29 {class=MaterialExpressionNamedRerouteUsage}
  MaterialExpressionNamedRerouteUsage_3 {class=MaterialExpressionNamedRerouteUsage}
  MaterialExpressionNamedRerouteUsage_30 {class=MaterialExpressionNamedRerouteUsage}
  MaterialExpressionNamedRerouteUsage_31 {class=MaterialExpressionNamedRerouteUsage}
  MaterialExpressionNamedRerouteUsage_32 {class=MaterialExpressionNamedRerouteUsage}
  MaterialExpressionNamedRerouteUsage_33 {class=MaterialExpressionNamedRerouteUsage}
  MaterialExpressionNamedRerouteUsage_34 {class=MaterialExpressionNamedRerouteUsage}
  MaterialExpressionNamedRerouteUsage_35 {class=MaterialExpressionNamedRerouteUsage}
  MaterialExpressionNamedRerouteUsage_36 {class=MaterialExpressionNamedRerouteUsage}
  MaterialExpressionNamedRerouteUsage_37 {class=MaterialExpressionNamedRerouteUsage}
  MaterialExpressionNamedRerouteUsage_38 {class=MaterialExpressionNamedRerouteUsage}
  MaterialExpressionNamedRerouteUsage_39 {class=MaterialExpressionNamedRerouteUsage}
  MaterialExpressionNamedRerouteUsage_4 {class=MaterialExpressionNamedRerouteUsage}
  MaterialExpressionNamedRerouteUsage_40 {class=MaterialExpressionNamedRerouteUsage}
  MaterialExpressionNamedRerouteUsage_41 {class=MaterialExpressionNamedRerouteUsage}
  MaterialExpressionNamedRerouteUsage_42 {class=MaterialExpressionNamedRerouteUsage}
  MaterialExpressionNamedRerouteUsage_44 {class=MaterialExpressionNamedRerouteUsage}
  MaterialExpressionNamedRerouteUsage_46 {class=MaterialExpressionNamedRerouteUsage}
  MaterialExpressionNamedRerouteUsage_47 {class=MaterialExpressionNamedRerouteUsage}
  MaterialExpressionNamedRerouteUsage_48 {class=MaterialExpressionNamedRerouteUsage}
  MaterialExpressionNamedRerouteUsage_5 {class=MaterialExpressionNamedRerouteUsage}
  MaterialExpressionNamedRerouteUsage_50 {class=MaterialExpressionNamedRerouteUsage}
  MaterialExpressionNamedRerouteUsage_51 {class=MaterialExpressionNamedRerouteUsage}
  MaterialExpressionNamedRerouteUsage_53 {class=MaterialExpressionNamedRerouteUsage}
  MaterialExpressionNamedRerouteUsage_54 {class=MaterialExpressionNamedRerouteUsage}
  MaterialExpressionNamedRerouteUsage_55 {class=MaterialExpressionNamedRerouteUsage}
  MaterialExpressionNamedRerouteUsage_56 {class=MaterialExpressionNamedRerouteUsage}
  MaterialExpressionNamedRerouteUsage_57 {class=MaterialExpressionNamedRerouteUsage}
  MaterialExpressionNamedRerouteUsage_58 {class=MaterialExpressionNamedRerouteUsage}
  MaterialExpressionNamedRerouteUsage_59 {class=MaterialExpressionNamedRerouteUsage}
  MaterialExpressionNamedRerouteUsage_60 {class=MaterialExpressionNamedRerouteUsage}
  MaterialExpressionNamedRerouteUsage_61 {class=MaterialExpressionNamedRerouteUsage}
  MaterialExpressionNamedRerouteUsage_62 {class=MaterialExpressionNamedRerouteUsage}
  MaterialExpressionNamedRerouteUsage_63 {class=MaterialExpressionNamedRerouteUsage}
  MaterialExpressionNamedRerouteUsage_64 {class=MaterialExpressionNamedRerouteUsage}
  MaterialExpressionNamedRerouteUsage_68 {class=MaterialExpressionNamedRerouteUsage}
  MaterialExpressionNamedRerouteUsage_69 {class=MaterialExpressionNamedRerouteUsage}
  MaterialExpressionNamedRerouteUsage_7 {class=MaterialExpressionNamedRerouteUsage}
  MaterialExpressionNamedRerouteUsage_70 {class=MaterialExpressionNamedRerouteUsage}
  MaterialExpressionNamedRerouteUsage_76 {class=MaterialExpressionNamedRerouteUsage}
  MaterialExpressionNamedRerouteUsage_77 {class=MaterialExpressionNamedRerouteUsage}
  MaterialExpressionNamedRerouteUsage_78 {class=MaterialExpressionNamedRerouteUsage}
  MaterialExpressionNamedRerouteUsage_79 {class=MaterialExpressionNamedRerouteUsage}
  MaterialExpressionNamedRerouteUsage_8 {class=MaterialExpressionNamedRerouteUsage}
  MaterialExpressionNamedRerouteUsage_80 {class=MaterialExpressionNamedRerouteUsage}
  MaterialExpressionNamedRerouteUsage_81 {class=MaterialExpressionNamedRerouteUsage}
  MaterialExpressionNamedRerouteUsage_82 {class=MaterialExpressionNamedRerouteUsage}
  MaterialExpressionNamedRerouteUsage_83 {class=MaterialExpressionNamedRerouteUsage}
  MaterialExpressionNamedRerouteUsage_84 {class=MaterialExpressionNamedRerouteUsage}
  MaterialExpressionNamedRerouteUsage_85 {class=MaterialExpressionNamedRerouteUsage}
  MaterialExpressionNamedRerouteUsage_86 {class=MaterialExpressionNamedRerouteUsage}
  MaterialExpressionNamedRerouteUsage_87 {class=MaterialExpressionNamedRerouteUsage}
  MaterialExpressionNamedRerouteUsage_88 {class=MaterialExpressionNamedRerouteUsage}
  MaterialExpressionNamedRerouteUsage_89 {class=MaterialExpressionNamedRerouteUsage}
  MaterialExpressionNamedRerouteUsage_9 {class=MaterialExpressionNamedRerouteUsage}
  MaterialExpressionNamedRerouteUsage_90 {class=MaterialExpressionNamedRerouteUsage}
  MaterialExpressionNamedRerouteUsage_91 {class=MaterialExpressionNamedRerouteUsage}
  MaterialExpressionNamedRerouteUsage_92 {class=MaterialExpressionNamedRerouteUsage}
  MaterialExpressionNamedRerouteUsage_93 {class=MaterialExpressionNamedRerouteUsage}
  MaterialExpressionNamedRerouteUsage_94 {class=MaterialExpressionNamedRerouteUsage}
  MaterialExpressionNormalize_2 {class=MaterialExpressionNormalize, inputs [VectorInput {from=MaterialExpressionNamedRerouteUsage_92, output=""}]}
  MaterialExpressionObjectPositionWS_1 {class=MaterialExpressionObjectPositionWS, inputs [], properties {}}
  MaterialExpressionPower_0 {
   class=MaterialExpressionPower
   inputs [
    Base {from=MaterialExpressionMaterialFunctionCall_13, output=Result}
    Exp {from=MaterialExpressionNamedRerouteUsage_78}
   ]
  }
  MaterialExpressionPower_2 {
   class=MaterialExpressionPower
   inputs [
    Base {from=MaterialExpressionMaterialFunctionCall_26, output=Result}
    Exp {from=MaterialExpressionNamedRerouteUsage_90}
   ]
  }
  MaterialExpressionPower_4 {
   class=MaterialExpressionPower
   inputs [
    Base {from=MaterialExpressionTextureSample_4, output=R}
    Exp {from=MaterialExpressionNamedRerouteUsage_31}
   ]
  }
  MaterialExpressionPower_6 {
   class=MaterialExpressionPower
   inputs [
    Base {from=MaterialExpressionNamedRerouteUsage_56}
    Exp {from=MaterialExpressionNamedRerouteUsage_86}
   ]
  }
  MaterialExpressionPower_8 {
   class=MaterialExpressionPower
   inputs [
    Base {from=MaterialExpressionMaterialFunctionCall_27, output=Result}
    Exp {from=MaterialExpressionNamedRerouteUsage_81}
   ]
  }
  MaterialExpressionReroute_0 {class=MaterialExpressionReroute, inputs [None {from=MaterialExpressionTextureSample_6, output=RGB}]}
  MaterialExpressionReroute_1 {class=MaterialExpressionReroute, inputs [None {from=MaterialExpressionTextureSampleParameter2D_4, output=RGB}]}
  MaterialExpressionReroute_4 {class=MaterialExpressionReroute, inputs [None {from=MaterialExpressionSubtract_1}]}
  MaterialExpressionSaturate_0 {class=MaterialExpressionSaturate, inputs [None {from=MaterialExpressionVectorParameter_13, output=Enable Haze}]}
  MaterialExpressionSaturate_10 {class=MaterialExpressionSaturate, inputs [None {from=MaterialExpressionMultiply_7}]}
  MaterialExpressionSaturate_4 {class=MaterialExpressionSaturate, inputs [None {from=MaterialExpressionVectorParameter_2, output=RGB}]}
  MaterialExpressionScalarParameter_0 {class=MaterialExpressionScalarParameter, properties {defaultValue=1, group=Ice | 04_Volume, parameterName=TailIceDensity}}
  MaterialExpressionScalarParameter_1 {class=MaterialExpressionScalarParameter, properties {defaultValue=1, group=Ice | 04_Volume, parameterName=CubeIceDensity}}
  MaterialExpressionScalarParameter_2 {class=MaterialExpressionScalarParameter, properties {group=00_Glass, parameterName=VertexHide, sortPriority=301}}
  MaterialExpressionScalarParameter_25 {
   class=MaterialExpressionScalarParameter
   properties {defaultValue=1.5, group=Ice | 03_Refraction, parameterName=Refraction - IOR, sliderMax=5, sliderMin=1, sortPriority=30}
  }
  MaterialExpressionScalarParameter_27 {
   class=MaterialExpressionScalarParameter
   properties {defaultValue=1, group=Ice | 03_Refraction, parameterName=Refraction - Surface Normal Mix, sliderMax=1, sortPriority=10}
  }
  MaterialExpressionScalarParameter_38 {
   class=MaterialExpressionScalarParameter
   properties {defaultValue=1, group=Ice | 02_Surface, parameterName=Surface - Detail Normal Intensity, sortPriority=30}
  }
  MaterialExpressionScalarParameter_39 {
   class=MaterialExpressionScalarParameter
   properties {defaultValue=1, group=Ice | 03_Refraction, parameterName=Refraction - Detail Normal Mix, sliderMax=1, sortPriority=20}
  }
  MaterialExpressionScalarParameter_48 {
   class=MaterialExpressionScalarParameter
   properties {defaultValue=1, group=Ice | 04_Volume, parameterName=Volume - Scattering Variation Speed, sortPriority=90}
  }
  MaterialExpressionScalarParameter_49 {
   class=MaterialExpressionScalarParameter
   properties {group=Ice | 04_Volume, parameterName=Volume - Scattering Variation Amount, sliderMax=1, sortPriority=60}
  }
  MaterialExpressionScalarParameter_76 {class=MaterialExpressionScalarParameter, properties {defaultValue=1, group=Ice | 04_Volume, parameterName=Volume - Density, sortPriority=20}}
  MaterialExpressionScalarParameter_77 {
   class=MaterialExpressionScalarParameter
   properties {group=Ice | 04_Volume, parameterName="Volume - Scattering Phase [PC]", sliderMax=1, sliderMin=-1, sortPriority=50}
  }
  MaterialExpressionScalarParameter_78 {
   class=MaterialExpressionScalarParameter
   properties {defaultValue=1, group=Ice | 03_Refraction, parameterName=Refraction - Transmission Attenuation, sortPriority=40}
  }
  MaterialExpressionSceneColor_0 {class=MaterialExpressionSceneColor, inputs [UVs {from=MaterialExpressionNamedRerouteUsage_18, output=""}], properties {}}
  MaterialExpressionSetMaterialAttributes_0 {
   class=MaterialExpressionSetMaterialAttributes
   inputs [
    MaterialAttributes {from=MaterialExpressionStaticSwitchParameter_3, output=""}
    World Position Offset {from=MaterialExpressionMultiply_5, output=""}
   ]
   properties {attributeSetTypes [#1]}
  }
  MaterialExpressionSetMaterialAttributes_5 {
   class=MaterialExpressionSetMaterialAttributes
   inputs [
    MaterialAttributes {from="", output=""}
    Base Color {from=MaterialExpressionNamedRerouteUsage_91, output=""}
    Emissive Color {from=MaterialExpressionNamedRerouteUsage_29, output=""}
    Metallic {from=MaterialExpressionConstant_3, output=""}
    Specular {from=MaterialExpressionNamedRerouteUsage_21, output=""}
    Roughness {from=MaterialExpressionNamedRerouteUsage_22, output=""}
    Normal {from=MaterialExpressionNamedRerouteUsage_69, output=""}
    Opacity {from=MaterialExpressionNamedRerouteUsage_10, output=""}
   ]
   properties {attributeSetTypes [
     #2
     #3
     #4
     #5
     #6
     #7
     #8
    ]}
  }
  MaterialExpressionShadingPathSwitch_2 {
   class=MaterialExpressionShadingPathSwitch
   inputs [
    Default {from=MaterialExpressionMultiply_12, output=""}
    Deferred {from="", output=""}
    Forward {from="", output=""}
    Mobile {from=MaterialExpressionMaterialFunctionCall_1, output="Scene Luminance (V3)"}
   ]
   properties {}
  }
  MaterialExpressionShadingPathSwitch_3 {
   class=MaterialExpressionShadingPathSwitch
   inputs [
    Default {from=MaterialExpressionMaterialFunctionCall_35, output="Refraction Result - AlphaBlending (V3)"}
    Deferred {from="", output=""}
    Forward {from="", output=""}
    Mobile {from=MaterialExpressionMaterialFunctionCall_35, output="Refraction Result - AlphaBlending (V3)"}
   ]
   properties {}
  }
  MaterialExpressionShadingPathSwitch_4 {
   class=MaterialExpressionShadingPathSwitch
   inputs [
    Default {from=MaterialExpressionMaterialFunctionCall_35, output="Material Opacity - AlphaBlending (S)"}
    Deferred {from="", output=""}
    Forward {from="", output=""}
    Mobile {from=MaterialExpressionMaterialFunctionCall_35, output="Material Opacity - AlphaBlending (S)"}
   ]
   properties {}
  }
  MaterialExpressionStaticSwitchParameter_0 {
   class=MaterialExpressionStaticSwitchParameter
   inputs [
    True {from=MaterialExpressionSetMaterialAttributes_0, output=""}
    False {from=MaterialExpressionMaterialFunctionCall_9, output=Result}
   ]
   properties {group=00_Glass, parameterName=UseVertexHide, sortPriority=300}
  }
  MaterialExpressionStaticSwitchParameter_3 {
   class=MaterialExpressionStaticSwitchParameter
   inputs [
    True {from=MaterialExpressionMaterialFunctionCall_6, output=Result}
    False {from=MaterialExpressionMaterialFunctionCall_3, output=Result}
   ]
   properties {group=00_Glass, parameterName=UseCameraDither, sortPriority=250}
  }
  MaterialExpressionSubtract_0 {
   class=MaterialExpressionSubtract
   inputs [
    A {from=MaterialExpressionObjectPositionWS_1}
    B {from=MaterialExpressionWorldPosition_1, output=XYZ}
   ]
  }
  MaterialExpressionSubtract_1 {
   class=MaterialExpressionSubtract
   inputs [
    A {from=MaterialExpressionNamedRerouteUsage_89}
    B {from=MaterialExpressionNamedRerouteUsage_85}
   ]
  }
  MaterialExpressionTextureCoordinate_0 {class=MaterialExpressionTextureCoordinate}
  MaterialExpressionTextureCoordinate_1 {class=MaterialExpressionTextureCoordinate}
  MaterialExpressionTextureCoordinate_2 {class=MaterialExpressionTextureCoordinate}
  MaterialExpressionTextureCoordinate_3 {class=MaterialExpressionTextureCoordinate}
  MaterialExpressionTextureCoordinate_4 {class=MaterialExpressionTextureCoordinate}
  MaterialExpressionTextureCoordinate_5 {class=MaterialExpressionTextureCoordinate}
  MaterialExpressionTextureCoordinate_6 {class=MaterialExpressionTextureCoordinate}
  MaterialExpressionTextureObjectParameter_3 {
   class=MaterialExpressionTextureObjectParameter
   inputs [
    UVs {from="", output=""}
    Apply View MipBias {from="", output=""}
   ]
   properties {
    group=Ice | 02_Surface
    parameterName=Surface - Roughness Map
    samplerType=SAMPLERTYPE_Masks
    sortPriority=40
    texture=/BRCosmetics/Characters/Player/Male/Medium/Bodies/M_MED_GlacialTrooper/Textures/T_IceSurface_R
   }
  }
  MaterialExpressionTextureObjectParameter_4 {
   class=MaterialExpressionTextureObjectParameter
   inputs [
    UVs {from="", output=""}
    Apply View MipBias {from="", output=""}
   ]
   properties {
    group=Ice | 02_Surface
    parameterName=Surface - Detail Normal Map
    samplerType=SAMPLERTYPE_Normal
    sortPriority=20
    texture=/BRCosmetics/Characters/Player/Male/Medium/Bodies/M_MED_GlacialTrooper/Textures/T_IceSurface_N
   }
  }
  MaterialExpressionTextureSampleParameter2D_0 {
   class=MaterialExpressionTextureSampleParameter2D
   inputs [
    UVs {from=MaterialExpressionTextureCoordinate_5, output=""}
    Apply View MipBias {from="", output=""}
   ]
   properties {
    parameterName=TailIceMask
    samplerType=SAMPLERTYPE_LinearColor
    texture=/BRCosmetics/Weapons/FORT_Melee/Pickaxe_FlakeSlide/Textures/T_FlakeSlide_Pickaxe_RefractionDistortionMask
   }
  }
  MaterialExpressionTextureSampleParameter2D_1 {
   class=MaterialExpressionTextureSampleParameter2D
   inputs [
    UVs {from=MaterialExpressionTextureCoordinate_1, output=""}
    Apply View MipBias {from="", output=""}
   ]
   properties {
    group=Ice | 01_Globals
    parameterName=Diffuse Texture with Alpha Mask
    samplerSource=SSM_Wrap_WorldGroupSettings
    sortPriority=10
    texture=&T_BlacksRGB
   }
  }
  MaterialExpressionTextureSampleParameter2D_2 {
   class=MaterialExpressionTextureSampleParameter2D
   inputs [
    UVs {from=MaterialExpressionTextureCoordinate_4, output=""}
    Apply View MipBias {from="", output=""}
   ]
   properties {
    group=Ice | 01_Globals
    parameterName=Normals
    samplerSource=SSM_Wrap_WorldGroupSettings
    samplerType=SAMPLERTYPE_Normal
    sortPriority=20
    texture=/Engine/EngineMaterials/FlatNormal
   }
  }
  MaterialExpressionTextureSampleParameter2D_3 {
   class=MaterialExpressionTextureSampleParameter2D
   inputs [
    UVs {from=MaterialExpressionTextureCoordinate_6, output=""}
    Apply View MipBias {from="", output=""}
   ]
   properties {
    parameterName=TailIceMask
    samplerType=SAMPLERTYPE_LinearColor
    texture=&T_FlakeSlide_Pickaxe_RefractionDistortionMask
   }
  }
  MaterialExpressionTextureSampleParameter2D_4 {
   class=MaterialExpressionTextureSampleParameter2D
   inputs [
    UVs {from="", output=""}
    Apply View MipBias {from="", output=""}
   ]
   properties {
    group=00_Glass
    parameterName=TechArtMask
    samplerType=SAMPLERTYPE_Masks
    sortPriority=-96
    texture=&T_BlackMask
   }
  }
  MaterialExpressionTextureSample_15 {
   class=MaterialExpressionTextureSample
   inputs [
    UVs {from=MaterialExpressionAdd_18}
    Tex {from=""}
    Apply View MipBias {from=""}
   ]
   properties {samplerSource=SSM_Wrap_WorldGroupSettings, texture=/BRCosmetics/Animation/Props/Gadgets_SatinCheddar/Materials/T_Perlin_Noise_M1}
  }
  MaterialExpressionTextureSample_4 {
   class=MaterialExpressionTextureSample
   inputs [
    UVs {from=MaterialExpressionNamedRerouteUsage_119}
    Tex {from=MaterialExpressionNamedRerouteUsage_164}
    Apply View MipBias {from=""}
   ]
   properties {samplerSource=SSM_Wrap_WorldGroupSettings}
  }
  MaterialExpressionTextureSample_5 {
   class=MaterialExpressionTextureSample
   inputs [
    UVs {from=MaterialExpressionAdd_17}
    Tex {from=""}
    Apply View MipBias {from=""}
   ]
   properties {samplerSource=SSM_Wrap_WorldGroupSettings, texture=&T_Perlin_Noise_M1}
  }
  MaterialExpressionTextureSample_6 {
   class=MaterialExpressionTextureSample
   inputs [
    UVs {from=MaterialExpressionNamedRerouteUsage_121}
    Tex {from=MaterialExpressionNamedRerouteUsage_165}
    Apply View MipBias {from=""}
   ]
   properties {samplerSource=SSM_Wrap_WorldGroupSettings}
  }
  MaterialExpressionTime_0 {class=MaterialExpressionTime}
  MaterialExpressionTransform_2 {class=MaterialExpressionTransform, inputs [None {from=MaterialExpressionNormalize_2}]}
  MaterialExpressionVectorParameter_0 {
   class=MaterialExpressionVectorParameter
   properties {
    group=Ice | 02_Surface
    parameterName=Surface - UV Controls
    sortPriority=10
    channelNames {a=N/A, b=V Offset, g=U Offset, r=UV Scale}
    defaultValue=<rgba 1 0 0 0>
   }
  }
  MaterialExpressionVectorParameter_1 {
   class=MaterialExpressionVectorParameter
   properties {group=Ice | 04_Volume, parameterName=Volume - Absorption Coeffs, sortPriority=30, defaultValue=<rgba 0 0 0 1>}
  }
  MaterialExpressionVectorParameter_10 {
   class=MaterialExpressionVectorParameter
   properties {
    group=Ice | 02_Surface
    parameterName=Surface - Specular Controls
    sortPriority=50
    channelNames {a=N/A, b=Roughness Mult, g=Roughness Exp, r=Specular}
    defaultValue=<rgba 0.5 1 1 0>
   }
  }
  MaterialExpressionVectorParameter_11 {
   class=MaterialExpressionVectorParameter
   properties {
    group=Ice | 04_Volume
    parameterName=Volume - Thickness Remap
    sortPriority=10
    channelNames {a=Thickness Exp, b=Thickness OldMax, g=Thickness OldMin, r=Thickness Remap Mix}
    defaultValue=<rgba 0 0 20 1>
   }
  }
  MaterialExpressionVectorParameter_13 {
   class=MaterialExpressionVectorParameter
   properties {
    group=Ice | 04_Volume
    parameterName="Volume - Haze Controls [MOBILE]"
    sortPriority=100
    channelNames {a=Haze Max Amount, b=Haze Min Amount, g=Roughness Exponent, r=Enable Haze}
    defaultValue=<rgba 0 2.5 0 1>
   }
  }
  MaterialExpressionVectorParameter_2 {
   class=MaterialExpressionVectorParameter
   properties {
    group=Ice | 03_Refraction
    parameterName="Refraction - Refracted Scene Avg Color [MOBILE]"
    sortPriority=60
    defaultValue=<rgba 0.18 0.18 0.18 0>
   }
  }
  MaterialExpressionVectorParameter_3 {
   class=MaterialExpressionVectorParameter
   properties {group=Ice | 04_Volume, parameterName=Volume - Scattering Coeffs, sortPriority=40, defaultValue=<rgba 0 0 0 1>}
  }
  MaterialExpressionVectorParameter_6 {
   class=MaterialExpressionVectorParameter
   properties {
    group=Ice | 03_Refraction
    parameterName="Refraction - Refracted Scene Scale [PC]"
    sortPriority=50
    channelNames {a=Intensity, b=Tint B, g=Tint G, r=Tint R}
    defaultValue=<rgba 1 1 1 1>
   }
  }
  MaterialExpressionVectorParameter_7 {
   class=MaterialExpressionVectorParameter
   properties {
    group=Ice | 04_Volume
    parameterName=Volume - Scattering Variation Pattern
    sortPriority=70
    channelNames {a=Pattern Exp, b=Pattern OldMax, g=Pattern OldMin, r=Pattern Scale}
    defaultValue=<rgba 5 0.1 0.8 2.2>
   }
  }
  MaterialExpressionVectorParameter_8 {
   class=MaterialExpressionVectorParameter
   properties {
    group=Ice | 04_Volume
    parameterName=Volume - Scattering Variation Depth Breakup
    sortPriority=80
    channelNames {a=Add Constant Depth, b=Breakup Pattern Mult, g=Breakup Pattern Exp, r=Breakup Pattern Relative Scale}
    defaultValue=<rgba 0.4 0.7 5 1.7>
   }
  }
  MaterialExpressionVectorParameter_9 {
   class=MaterialExpressionVectorParameter
   properties {
    group=Ice | 02_Surface
    parameterName=Surface - Opacity Map Controls
    sortPriority=60
    channelNames {a=N/A, b=Opacity Exp, g=Opacity Old Max, r=Opacity Old Min}
    defaultValue=<rgba 1 1 1 0>
   }
  }
  MaterialExpressionWorldPosition_1 {class=MaterialExpressionWorldPosition, inputs [], properties {}}
 }
 properties {
  bAutomaticallySetUsageInEditor=false
  bEnableResponsiveAA=true
  bScreenSpaceReflections=true
  bUseMaterialAttributes=true
  bUsedWithMeshDeformer=true
  bUsedWithMorphTargets=true
  bUsedWithSkeletalMesh=true
  blendMode=BLEND_Translucent
  ditherOpacityMask=true
  refractionMethod=RM_IndexOfRefraction
  translucencyDirectionalLightingIntensity=5
  translucencyLightingMode=TLM_SurfacePerPixelLighting
  usedShadingModels=Default Lit
 }
}
