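# Opaque character-parent material for the FlakeSlide pickaxe that adds a fake refractive ice look in masked areas.
# The camera vector refracts through the intensity-scaled PreFX normal by inverse IOR, then samples environment illuminance with adjustable roughness and squared tint as emissive.
# This ice layer, with black non-metallic base and glossy specular, blends over PreFX attributes via a Tech Art Mask channel (blue by default) before PostFX.
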Material /BRCosmetics/Weapons/FORT_Melee/Pickaxe_FlakeSlide/Materials/M_Pickaxe_FlakeSlide_Ice_Opaque
graph {
 properties {
  bUseMaterialAttributes=true
  bUsedWithMeshDeformer=true
  bUsedWithNiagaraMeshParticles=true
  bUsedWithSkeletalMesh=true
  pixelDepthOffsetMode=PDOM_Legacy
  shadingModel=MSM_FromMaterialExpression
  usedShadingModels=Default Lit | Subsurface | Clear Coat | Subsurface Profile | Cloth
 }
 nodes [
  n1 = MaterialFunctionCall() {materialFunction=&MF_CharacterParent_2023_PreFX}
  n2 = GetMaterialAttributes(None=@n1.Result) {attributeGetTypes [#1]}
  n3 = ScalarParameter() {defaultValue=1.5, group=Refraction, parameterName=Refraction - IOR, sliderMax=5, sliderMin=1, sortPriority=20}
  n4 = MaterialFunctionCall!(
   View Vector=CameraVectorWS()
   Normal=MaterialFunctionCall!(
    Normal=NamedRerouteUsage()
    Intensity=ScalarParameter() {defaultValue=1, group=Refraction, parameterName=Refraction - Normal Intensity, sliderMax=1, sortPriority=10}
   ) {materialFunction=/Game/Global/MaterialFunctions/MF_NormalIntensity}."Result (V3)"
   "Relative IOR (Eta)"=Divide(B=Clamp(None=@n3) {maxDefault=5, minDefault=1})
  ) {materialFunction=/Game/Global/MaterialFunctions/TechArt/Refraction/MF_Refract}
  n5 = ScalarParameter() {defaultValue=0.2, group=Refraction, parameterName=Refraction - Roughness, sliderMax=1, sortPriority=30}
  n6 = VectorParameter() {
   group=Refraction
   parameterName=Refraction - Tint
   sortPriority=40
   channelNames {a=Brightness, b=Tint B, g=Tint G, r=Tint R}
   defaultValue=<rgba 1 1 1 1>
  }
  n7 = Constant()
  n8 = SetMaterialAttributes!(
   MaterialAttributes
   Emissive Color=Multiply(
    A=MaterialFunctionCall!(
     Direction WS=@n4."Refraction Vector (V3)"
     Roughness=@n5
     Sky Desaturation
     Sky Brightness
     Use Ground Map
     Ground LatLong Map
     Group Map Rotation
     Ground Exp
     Ground Brightness
    ) {materialFunction=/Game/Global/MaterialFunctions/TechArt/Refraction/MF_EnvironmentIlluminance}."Environment Illuminance (V3)"
    B=Multiply(A=@n6.RGB, B=@n6.RGB)
   )
   Base Color=@n7
   Metallic=@n7
   Specular=ScalarParameter() {defaultValue=0.5, group=Specular, parameterName=Specular, sliderMax=1, sortPriority=10}
   Roughness=ScalarParameter() {defaultValue=0.1, group=Specular, parameterName=Roughness, sliderMax=1, sortPriority=20}
   Normal=NamedRerouteUsage()
   Shading Model=ShadingModel()
  ) {attributeSetTypes [
    #2
    #3
    #4
    #5
    #6
    #1
    #7
   ]}
  MaterialFunctionCall(
   Material=BlendMaterialAttributes(
    A=@n2.MaterialAttributes
    B=@n8
    Alpha=ChannelMaskParameter(None=@n1.Tech Art Mask) {group=SimpleIce, maskChannel=Blue, parameterName=IceFXMask, defaultValue=<rgba 0 0 1 0>}.R
   )
   M=@n1.M
   Pre Skinned Local Position=@n1.M
   Distance Blend=@n1.M
   Fresnel=@n1.M
  ) {materialFunction=&MF_CharacterParent_2023_PostFX}
  NamedRerouteDeclaration(None=@n2.Normal) {name="Refraction Normal In (V3)", nodeColor=<rgba 1 0 0.7031 1>}
 ]
}
