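# Physics asset for the FlakeSlide two-handed pickaxe mesh, defining the bodies and joints for its dangling strap.
# A large capsule covers the handle bone, and two small capsules cover the strap bones.
# The strap bones form a chain off the handle, with constraints that lock one swing axis and tightly limit twist.
# Each joint therefore bends mostly in one plane, with a wider range at the root than at the second segment.
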PhysicsAsset /BRCosmetics/Weapons/FORT_Melee/Pickaxe_FlakeSlide/Meshes/FlakeSlide_Axe_Physics
bodies {
 Melee_TwoHanded_Handle [{shapeType=Capsule, length=18.549, radius=20.6268, center=<-8.9435 -0.0026 -10.5118>, rotation=<rot 0.0028 -90 -0.2272>}]
 dyn_strap_01 [{
   shapeType=Capsule
   length=1.5805
   radius=1.9739
   shapeName=dyn_strap_01_capsule
   center=<3.2461 -0.026 -0.028>
   rotation=<rot 0 -88.4023 -90.1666>
  }]
 dyn_strap_02 [{shapeType=Capsule, length=4.5771, radius=1.6522, shapeName=dyn_strap_02_capsule, center=<4.5372 0 0>, rotation=<rot 0 -89.9998 90>}]
}
constraints [
 {bone1Name=dyn_strap_01, bone2Name=Melee_TwoHanded_Handle, swing1LimitDegrees=0, swing2LimitDegrees=90, twistLimitDegrees=5}
 {bone1Name=dyn_strap_02, bone2Name=dyn_strap_01, swing1LimitDegrees=0, swing2LimitDegrees=45, twistLimitDegrees=5}
]
